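; Blueprint for the Athena EMP grenade item, likely the held or thrown grenade actor.
; On BeginPlay it runs parent logic, then sets an "IsLocal" bool on its in-hand Niagara effect from whether the instigator pawn is locally controlled.
; The construction script only calls the parent.
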(event EventBeginPlay
  (|CallParentFunction)
  (Effects|Components|ParticleSystem|SetBoolParameter (Variables|Default|GetNS_EMPGrenade_InHand) "IsLocal" (Pawn|IsLocallyControlled (Game|GetInstigator self))))

(fn ConstructionScript
  (|CallParentFunction))
