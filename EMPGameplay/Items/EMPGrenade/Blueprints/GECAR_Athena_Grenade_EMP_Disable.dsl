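; Custom application requirement that decides whether the EMP grenade's disable gameplay effect can apply to a target.
; It casts the target ability system component's owner to the Fort vehicle interface and checks the component for the EMPCanBeApplied tag.
; Either a successful cast or that tag qualifies the target, but the EMPCanNotBeApplied tag always blocks the effect.
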(fn CanApplyGameplayEffect
  (params (GameplayEffect :type Gameplay%20Effect%20Object%20Reference) (Spec :type Gameplay%20Effect%20Spec%20Structure%20%28by%20ref%29) (ASC :type Ability%20System%20Component%20Object%20Reference))
  (bind (_asfort_vehicle_interface _bsuccess) (Utilities|Casting|CastToFortVehicleInterface (Components|GetOwner ASC)))
  (return (and (or (GameplayTags|HasMatchingGameplayTag ASC (Variables|Default|GetTag_EMPCanBeApplied)) _bsuccess) (not (GameplayTags|HasMatchingGameplayTag ASC (Variables|Default|GetTag_EMPCanNotBeApplied))))))
