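; EMP grenade projectile that spins its explosive mesh in flight and explodes at its own location on hit or stop.
; Tick advances the mesh's relative rotation by a rotation rate, and water entry calls a water impact for the first body or explodes otherwise.
; Explosion and stop events deactivate the handheld Niagara effect and hide the mesh, with explosion also halting movement.
; Construction switches that effect to projectile mode with boosted light radius.
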(event EventTick
  (params (DeltaSeconds :type Float%20%28single-precision%29))
  (bind (_relativerotation_roll _relativerotation_pitch _relativerotation_yaw) (Math|Rotator|BreakRotator (Variables|/Script/Engine.SceneComponent|GetRelativeRotation_Roll (Variables|Default|GetGrenade_EMP_Explosive))))
  (bind (_rotationrate_roll _rotationrate_pitch _rotationrate_yaw) (Math|Rotator|BreakRotator (Variables|GetRotationRate_Roll)))
  ;[1 1 1] On Tick
  (|CallParentFunction DeltaSeconds)
  (Transformation|SetRelativeRotation (Variables|Default|GetGrenade_EMP_Explosive) (Math|Rotator|MakeRotator (+ _relativerotation_roll (* DeltaSeconds _rotationrate_roll)) (+ _relativerotation_pitch (* DeltaSeconds _rotationrate_pitch)) (+ _relativerotation_yaw (* DeltaSeconds _rotationrate_yaw))) false false))

(event Collision|EventHit
  (params (MyComp :type Primitive%20Component%20Object%20Reference) (Other :type Actor%20Object%20Reference) (OtherComp :type Primitive%20Component%20Object%20Reference) (bSelfMoved :type Boolean) (HitLocation :type Vector :default "0, 0, 0") (HitNormal :type Vector :default "0, 0, 0") (NormalImpulse :type Vector :default "0, 0, 0") (Hit :type Hit%20Result%20Structure%20%28by%20ref%29))
  ;[1 1 1] On Hit
  (|CallParentFunction MyComp Other OtherComp bSelfMoved HitLocation HitNormal NormalImpulse Hit)
  (Delivery|DoExplosionatLocation (Transformation|GetActorLocation)))

(event OnEnterWaterEvent(WaterInteractionComponent)
  (params (WaterBody :type Fort%20Water%20Body%20Actor%20Object%20Reference) (WaterInteractionComponent :type Fort%20Water%20Interaction%20Component%20Object%20Reference) (bIsFirstBody :type Boolean))
  ;[1 1 1] On Enter Water
  (if bIsFirstBody
    (CallFunction|HeldWaterImpact)
    (Delivery|DoExplosionatLocation (Transformation|GetActorLocation))))

(event EventOnExploded
  (params (HitActors :type Array%20of%20Actor%20Object%20References) (HitResults :type Array%20of%20Hit%20Result%20Structures))
  ;[1 1 1] On Exploded
  (|CallParentFunction HitActors HitResults)
  (Components|Activation|Deactivate (Variables|Default|GetNS_HandHeldFX))
  (Development|SetHiddeninGame (Variables|Default|GetGrenade_EMP_Explosive) true false)
  (Components|Movement|StopMovementImmediately (Variables|Movement|GetProjectileMovementComponent)))

(event EventOnStoppedOnServer
  ;[1 1 1] On Stop (Server)
  (|CallParentFunction)
  (Delivery|DoExplosionatLocation (Transformation|GetActorLocation)))

(event EventOnStop
  (params (Hit :type Hit%20Result%20Structure%20%28by%20ref%29))
  ;[1 1 1] On Stop
  (|CallParentFunction Hit)
  (Delivery|DoExplosionatLocation (Transformation|GetActorLocation))
  (Components|Activation|Deactivate (Variables|Default|GetNS_HandHeldFX))
  (Development|SetHiddeninGame (Variables|Default|GetGrenade_EMP_Explosive) true false))

(fn ConstructionScript
  (bind _ns_handheldfx (Variables|Default|GetNS_HandHeldFX))
  (|CallParentFunction)
  (Effects|Components|ParticleSystem|SetFloatParameter _ns_handheldfx "LightRadiusMulti" 8.0)
  (Effects|Components|ParticleSystem|SetBoolParameter _ns_handheldfx "IsPrj" true))
