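# Unlit translucent material for Niagara mesh particles, rendering the EMP grenade's glowing energy core.
# A panning plasma texture modulates the exponent on a view-angle fade, which blends three colors.
# The boosted result is multiplied by particle color and passed through emissive adjustments.
# UseFadeBOVA, zero by default, blends opacity from plain particle alpha to a depth-faded, view-angle-faded version.
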Material /EMPGameplay/Items/EMPGrenade/FX/M_EMPGrenade_Energy_Core
graph {
 properties {
  bUsedWithNiagaraMeshParticles=true
  blendMode=BLEND_Translucent
  pixelDepthOffsetMode=PDOM_Legacy
  shadingModel=MSM_Unlit
  usedShadingModels=Unlit
 }
 nodes [
  n1 = ParticleColor()
  n2 = Constant() {r=4}
  n3 = Saturate(None=Multiply(A=MaterialFunctionCall("FadeSensitivity (optional)"=@n2) {materialFunction=&FadeBasedOnViewAngle}.Result) {constB=10})
  LinearInterpolate(A=@n1.A, B=DepthFade(Opacity=Multiply(A=@n1.A, B=@n3)) {fadeDistanceDefault=300}, Alpha=ScalarParameter() {parameterName=UseFadeBOVA})
  n4 = TextureSample(UVs=Panner(Coordinate=TextureCoordinate() {vTiling=0.5}) {speedX=0.5, speedY=0.5}) {samplerType=SAMPLERTYPE_LinearGrayscale, texture=/Game/Effects/Fort_Effects/Emotes/UltraLight/Texture/T_Plasma_A_Greyscale}
  n5 = MaterialFunctionCall(
   A=VectorParameter() {parameterName=ColorLow, defaultValue=<rgba 2.5432 2.6253 5 1>}.RGB
   B=VectorParameter() {parameterName=ColorHigh, defaultValue=<rgba 0.3149 0.0928 1 1>}.RGB
   C=VectorParameter() {defaultValue=<rgba 0 0.0222 0.0365 1>}.RGB
   Alpha=Power(
    Base=MaterialFunctionCall() {materialFunction=&FadeBasedOnViewAngle}.Result
    Exp=LinearInterpolate(Alpha=@n4.R) {constA=0.001, constB=0.4}
   )
  ) {materialFunction=&3ColorBlend}
  n6 = Constant() {r=5}
  MaterialFunctionCall(Emissive=Multiply(A=@n1.RGB, B=Multiply(A=@n5.Result, B=@n6))) {materialFunction=&MF_FX_EmissiveAdjustments}
  ScalarParameter() {defaultValue=0.2, parameterName=s}
 ]
}
