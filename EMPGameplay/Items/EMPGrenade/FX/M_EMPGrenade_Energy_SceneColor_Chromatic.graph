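# Translucent unlit Niagara sprite material for the EMP grenade's energy effect, distorting the scene behind it with chromatic separation.
# A panned normal map in radial coordinates offsets screen UVs for three SceneColor samples at decreasing strengths, recombined as red, green and boosted blue.
# A soft sphere mask with depth fade shapes the opacity.
# Shading path switches zero both color and opacity on Forward and Mobile.
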Material /EMPGameplay/Items/EMPGrenade/FX/M_EMPGrenade_Energy_SceneColor_Chromatic
graph {
 properties {bUsedWithNiagaraSprites=true, blendMode=BLEND_Translucent, pixelDepthOffsetMode=PDOM_Legacy, shadingModel=MSM_Unlit, usedShadingModels=Unlit}
 nodes [
  Constant3Vector() {constant=<rgba 1 0 0 0>}
  Constant()
  n1 = ScreenPosition()
  n2 = ScalarParameter() {defaultValue=2, parameterName=Param_1}
  n3 = ScalarParameter() {defaultValue=0.1, parameterName=Param_2}
  n4 = Panner(
   Coordinate=Multiply(
    A=MaterialFunctionCall() {materialFunction=&VectorToRadialValue}.Radial Coordinates
    B=AppendVector(A=@n2, B=@n3)
   )
   Speed=AppendVector(A=ScalarParameter() {parameterName=s}, B=ScalarParameter() {defaultValue=-0.2})
  ) {speedY=0.2}
  n5 = Divide(
   A=ComponentMask(None=TextureSample(UVs=@n4) {samplerType=SAMPLERTYPE_Normal, texture=/Game/Global/Textures/Default/Tileable_Patterns/1Normal}.RGB) {g=true, r=true}
   B=ScalarParameter() {defaultValue=10, parameterName=Param_3}
  ) {constB=6}
  n6 = Add(A=@n1.ViewportUV, B=@n5)
  n7 = Add(A=@n1.ViewportUV, B=Divide(A=@n5))
  n8 = SceneColor!(UVs=Add(A=@n1.ViewportUV, B=Divide(A=@n5) {constB=3}))
  n9 = MaterialFunctionCall(
   R=MaterialFunctionCall(Float3=SceneColor(UVs=@n6).RGB) {materialFunction=&BreakOutFloat3Components}.R
   G=MaterialFunctionCall(Float3=SceneColor(UVs=@n7).RGB) {materialFunction=&BreakOutFloat3Components}.G
   B=Multiply(
    A=MaterialFunctionCall(Float3=@n8.RGB) {materialFunction=&BreakOutFloat3Components}.B
    B=ScalarParameter() {defaultValue=4, parameterName=Param_4}
   ) {constB=5}
  ) {materialFunction=&AppendMany}
  n10 = Constant()
  ShadingPathSwitch(Default=@n9.RGB, Deferred=@n9.RGB, Forward=@n10, Mobile=@n10)
  n11 = TextureCoordinate()
  n12 = Constant() {r=0.5}
  n13 = DepthFade(Opacity=Power(Base=SphereMask(A=@n11, B=@n12, Radius=@n12) {hardnessPercent=0}))
  n14 = Constant()
  ShadingPathSwitch(Default=@n13, Deferred=@n13, Forward=@n14, Mobile=@n14)
  Subtract(A=@n1.ViewportUV, B=@n5)
  TextureCoordinate()
 ]
}
