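# Translucent unlit material for EMP grenade energy splines on Niagara mesh particles.
# World position offset combines tapered SplineThicken width with scrolling, per-particle randomized noise displacement, masked along the V gradient.
# Opacity shapes UV gradients with optional panning plasma erosion and a sine grow mask, then particle alpha and depth fade.
# Emissive blends two colors along V, scaled by brightness and particle color, while dynamic parameters drive width, wobble, growth and erosion.
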Material /EMPGameplay/Items/EMPGrenade/FX/M_EMPGrenade_Energy_Spline
graph {
 properties {
  bUsedWithNiagaraMeshParticles=true
  blendMode=BLEND_Translucent
  pixelDepthOffsetMode=PDOM_Legacy
  shadingModel=MSM_Unlit
  usedShadingModels=Unlit
 }
 nodes [
  n1 = MaterialFunctionCall() {materialFunction=&LinearGradient}
  n2 = Multiply(A=@n1.VGradient) {constB=0.5}
  n3 = Power(Base=OneMinus(None=Sine(None=@n2))) {constExponent=4}
  n4 = Multiply(
   A=ScalarParameter() {defaultValue=2, parameterName=SplineWidth}
   B=DynamicParameter() {paramNames [SplineWidth,WPOIntensity,GrowMask,Erode]}.SplineWidth
  )
  n5 = Power(
   Base=MaterialFunctionCall() {materialFunction=&LinearGradient}.VGradient
   Exp=ScalarParameter() {defaultValue=0.7, parameterName=Param_3}
  )
  n6 = Constant() {r=0.5}
  n7 = Time()
  n8 = LinearInterpolate(Alpha=ParticleRandom()) {constA=0.3}
  n9 = Add(A=Multiply(A=TextureCoordinate(), B=Constant2Vector() {g=0.7, r=1}), B=Add(A=ParticleRandom(), B=Multiply(A=@n7, B=@n8)))
  n10 = MaterialFunctionCall(Float3=TextureSampleParameter2D(UVs=@n9) {parameterName=WPO_Tex, texture=/Game/Global/Textures/Default/Noise/T_Noise_Cloud}.RGB) {materialFunction=&BreakOutFloat3Components}
  n11 = Transform(None=MaterialFunctionCall(R=LinearInterpolate(Alpha=@n10.R) {constA=-1}, G=LinearInterpolate(Alpha=@n10.G) {constA=-1}, B=LinearInterpolate(Alpha=@n10.B) {constA=-1}) {materialFunction=&AppendMany}.RGB)
  n12 = Multiply(
   A=ScalarParameter() {defaultValue=2, parameterName=Param_2}
   B=DynamicParameter() {paramNames [SplineWidth,WPOIntensity,GrowMask,Erode]}.WPOIntensity
  )
  Add(
   A=Multiply(
    A=LinearInterpolate(Alpha=@n3) {constA=0.35}
    B=MaterialFunctionCall(WidthBase=@n4, WidthTip=@n4) {materialFunction=&SplineThicken}.WorldPositionOffset
   )
   B=Multiply(A=SphereMask(A=@n5, B=@n6) {attenuationRadius=0.5, hardnessPercent=0}, B=Multiply(A=@n11, B=@n12)) {constA=1}
  ) {constB=0}
  n13 = ParticleColor()
  n14 = Multiply(A=@n1.UGradient) {constB=0.5}
  n15 = Multiply(A=Power(Base=Sine(None=@n14)) {constExponent=4}, B=Saturate(None=Multiply(A=@n1.VGradient) {constB=10}))
  n16 = TextureCoordinate()
  n17 = ParticleRandom()
  n18 = VectorParameter() {parameterName=Param_2, defaultValue=<rgba 0.4 0.3 0 0>}
  n19 = Panner(Coordinate=Multiply(A=Add(A=@n16, B=@n17), B=AppendVector(A=@n18.R, B=@n18.R))) {speedY=0.5}
  n20 = Subtract(
   A=Power(Base=TextureSample(UVs=@n19) {samplerType=SAMPLERTYPE_LinearGrayscale, texture=/Game/Effects/Fort_Effects/Emotes/UltraLight/Texture/T_Plasma_C_Greyscale}.R)
   B=DynamicParameter() {paramNames [SplineWidth,WPOIntensity,GrowMask,Erode]}.Erode
  )
  n21 = LinearInterpolate(A=@n15, B=Multiply(A=@n15, B=Saturate(None=@n20)), Alpha=ScalarParameter() {parameterName=UseDetailNoise})
  n22 = Divide(A=MaterialFunctionCall() {materialFunction=&LinearGradient}.VGradient) {constB=5}
  n23 = LinearInterpolate(Alpha=DynamicParameter() {paramNames [SplineWidth,WPOIntensity,GrowMask,Erode]}.GrowMask) {constA=-0.2, constB=0.5}
  n24 = Multiply(A=@n21, B=Sine(None=Add(A=@n22, B=@n23)))
  n25 = ScalarParameter() {parameterName=UsePanMask}
  DepthFade(Opacity=Multiply(A=@n13.A, B=LinearInterpolate(A=@n21, B=@n24, Alpha=@n25) {constAlpha=0})) {fadeDistanceDefault=20}
  n26 = LinearInterpolate(
   A=VectorParameter() {parameterName=Param_1, defaultValue=<rgba 0.2312 0.5423 3 1>}.RGB
   B=VectorParameter() {defaultValue=<rgba 0.1102 0.5971 1 1>}.RGB
   Alpha=Power(Base=@n1.VGradient) {constExponent=0.5}
  )
  n27 = ScalarParameter() {defaultValue=20, parameterName=Brightness}
  MaterialFunctionCall(Emissive=Multiply(A=Multiply(A=@n26, B=@n27), B=@n13.RGB)) {materialFunction=&MF_FX_EmissiveAdjustments}
  MaterialFunctionCall() {materialFunction=&DebugTimeSine}
  TextureSample(UVs=@n9) {texture=/Game/Effects/Fort_Effects/Emotes/UltraLight/Texture/T_SoftNoise_BC7}
 ]
}
