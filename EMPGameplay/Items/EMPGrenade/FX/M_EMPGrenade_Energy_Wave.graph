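# Additive unlit material for Niagara mesh particles that draws the EMP grenade's energy wave.
# SplineThicken sets width, and scrolling, per-particle-offset noise adds transformed world position displacement.
# A curve-atlas gradient times two multiplied panning noise samples forms a mask for depth-faded opacity and a refraction lerp, alongside tinted, particle-colored emissive.
# Dynamic parameters drive spline width, displacement intensity, erosion and refraction; several texture samples are left unconnected.
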Material /EMPGameplay/Items/EMPGrenade/FX/M_EMPGrenade_Energy_Wave
graph {
 properties {
  bUsedWithNiagaraMeshParticles=true
  blendMode=BLEND_Additive
  pixelDepthOffsetMode=PDOM_Legacy
  refractionMethod=RM_PixelNormalOffset
  shadingModel=MSM_Unlit
  usedShadingModels=Unlit
 }
 nodes [
  n1 = TextureCoordinate()
  n2 = Multiply(
   A=ScalarParameter() {defaultValue=20}
   B=DynamicParameter() {defaultValue=<rgba 1 1 0 1>, paramNames [SplineWidth,WPOIntensity,Erode,Refraction]}.SplineWidth
  )
  n3 = VectorParameter() {parameterName=WPO_UV, defaultValue=<rgba 0.1 4 0 0>}
  n4 = MaterialFunctionCall(Float2=Multiply(A=TextureCoordinate(), B=ComponentMask(None=@n3.RGB) {g=true, r=true})) {materialFunction=&BreakOutFloat2Components}
  n5 = ParticleRandom()
  n6 = Multiply(A=Time(), B=LinearInterpolate(Alpha=ParticleRandom()) {constA=0.3})
  n7 = AppendVector(A=Add(A=@n4.R, B=Add(A=@n5, B=@n6)), B=@n4.G)
  n8 = MaterialFunctionCall(Float3=TextureSample(UVs=@n7) {texture=&2ndInfectedMapLinear}.RGB) {materialFunction=&BreakOutFloat3Components}
  n9 = MaterialFunctionCall(R=LinearInterpolate(Alpha=@n8.R) {constA=-1}, G=Constant(), B=LinearInterpolate(Alpha=@n8.B) {constA=-1}) {materialFunction=&AppendMany}
  n10 = ScalarParameter() {defaultValue=5, parameterName=WPO_Intensity}
  n11 = DynamicParameter() {defaultValue=<rgba 1 1 0 1>, paramNames [SplineWidth,WPOIntensity,Erode,Refraction]}
  Add(
   A=MaterialFunctionCall(WidthBase=@n2, WidthTip=@n2, UVs for Projection=@n1, UVs for Thickness=@n1, UVs for texturing=@n1) {materialFunction=&SplineThicken}.WorldPositionOffset
   B=Multiply(A=Transform(None=@n9.RGB), B=Multiply(A=@n10, B=@n11.WPOIntensity))
  ) {constB=0}
  LinearInterpolate(Alpha=@n8.G) {constA=-1}
  n12 = CurveAtlasRowParameter(CurveTime=MaterialFunctionCall() {materialFunction=&LinearGradient}.UGradient) {
   atlas=/Game/Athena/LinearColorCurves/CA_Contrail_SilentTempo
   curve=/Game/Accessories/FORT_Backpacks/Backpack_F_MED_FishBowlBone/FX/Curve_LeadEdge
   defaultValue=14
   parameterName=Param_1
  }
  n13 = VectorParameter() {parameterName=Noise_UV, defaultValue=<rgba 0.1 2 0 0>}
  n14 = Panner(Coordinate=Multiply(A=TextureCoordinate(), B=ComponentMask(None=@n13.RGB) {g=true, r=true})) {speedX=0.2, speedY=0.2}
  n15 = Multiply(
   A=TextureSample(UVs=Divide(A=@n14)) {samplerType=SAMPLERTYPE_Grayscale, texture=&LowResBlurredNoise}.R
   B=TextureSample(UVs=@n14) {samplerType=SAMPLERTYPE_Grayscale, texture=&LowResBlurredNoise}.R
  )
  n16 = Multiply(A=@n12.R, B=Multiply(A=Power(Base=@n15), B=ScalarParameter() {defaultValue=5, parameterName=Param_6}))
  LinearInterpolate(Alpha=Multiply(A=@n16, B=DynamicParameter() {defaultValue=<rgba 1 1 0 1>, paramNames [SplineWidth,WPOIntensity,Erode,Refraction]}.Refraction)) {constA=1, constB=2}
  n17 = VectorParameter() {parameterName=Color, defaultValue=<rgba 0.0466 0.108 1 1>}
  n18 = ScalarParameter() {defaultValue=30, parameterName=Param_4}
  n19 = ParticleColor()
  MaterialFunctionCall(Emissive=Multiply(A=Multiply(A=@n17.RGB, B=@n18), B=@n19.RGB)) {materialFunction=&MF_FX_EmissiveAdjustments}
  n20 = DynamicParameter() {defaultValue=<rgba 1 1 0 1>, paramNames [SplineWidth,WPOIntensity,Erode,Refraction]}
  Multiply(A=@n19.A, B=DepthFade(Opacity=Subtract(A=@n16, B=@n20.Erode)))
  TextureSample(UVs=@n7) {samplerType=SAMPLERTYPE_Grayscale, texture=&LowResBlurredNoise}
  TextureSample(UVs=TextureCoordinate() {uTiling=0.5, vTiling=2}) {samplerType=SAMPLERTYPE_Normal, texture=/Game/Characters/Player/Female/Medium/Bodies/F_MED_Bistro_Astronaut/FX/T_GaussianNoise_Normal}
  TextureSample(UVs=@n7) {samplerType=SAMPLERTYPE_Normal, texture=&T_GaussianNoise_Normal}
  TextureSample() {texture=/Game/Effects/Fort_Effects/Emotes/UltraLight/Texture/T_SoftNoise_BC7}
  TextureSample() {samplerType=SAMPLERTYPE_LinearGrayscale, texture=/Game/Environments/Hexmap/HexClouds/T_HexmapFog_01_Height02}
  TextureSample(UVs=@n7) {samplerType=SAMPLERTYPE_Masks, texture=/Game/Global/Textures/Default/Noise/T_CellularNoise}
 ]
}
