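# M_EMPGrenade_ShieldCrack is an unlit additive Niagara sprite material, named for the shield crack effect of the EMP grenade.
# It subtracts the red channel of a dynamic parameter from the red channel of a cracked decal texture and saturates the result, so particles can erode the crack pattern.
# That mask is multiplied by particle alpha for fading.
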Material /EMPGameplay/Items/EMPGrenade/FX/M_EMPGrenade_ShieldCrack
graph {
 properties {bUsedWithNiagaraSprites=true, blendMode=BLEND_Additive, pixelDepthOffsetMode=PDOM_Legacy, shadingModel=MSM_Unlit, usedShadingModels=Unlit}
 nodes [
  n1 = TextureSample() {texture=/Game/Effects/Fort_Effects/Gameplay/SmokeDustExplosions/Textures/T_Decal_Cracked}
  n2 = DynamicParameter()
  Multiply(A=ParticleColor().A, B=Saturate(None=Subtract(A=@n1.R, B=@n2.R)))
 ]
}
