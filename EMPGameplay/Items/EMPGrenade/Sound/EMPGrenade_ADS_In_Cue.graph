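# Sound cue for the EMP grenade's aim-down-sights start, layering clothing foley with cook and glass sounds.
# A mixer blends three layers: randomized run foley, a random grenade cook wave and a random glass cook wave, each with its own volume.
# The foley layer switches by quality level, with the last two levels drawing from only two waves.
# Attenuation uses the shared grenade cook settings, and concurrency allows one instance, stopping the oldest.
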SoundCue /EMPGameplay/Items/EMPGrenade/Sound/EMPGrenade_ADS_In_Cue
root=Mixer_0
nodes [
 {id=Mixer_0, children [Modulator_0,Modulator_1,Modulator_2], properties {inputVolume [0.4,0.5,0.2]}}
 {id=Modulator_0, children [QualityLevel_0], properties {volumeMax=0.8, volumeMin=0.8}}
 {id=Modulator_1, children [Random_0], properties {pitchMax=1, volumeMax=2.5, volumeMin=2.5}}
 {id=Modulator_2, children [Random_1], properties {pitchMax=1, volumeMax=2.5, volumeMin=2.5}}
 {id=QualityLevel_0, children [Random_4,Random_4,Random_2,Random_2]}
 {id=Random_0, children [WavePlayer_4,WavePlayer_5], properties {weights [1,1]}}
 {id=Random_1, children [WavePlayer_6,WavePlayer_7,WavePlayer_8], properties {weights [1,1,1]}}
 {id=Random_2, children [WavePlayer_2,WavePlayer_3], properties {bShouldExcludeFromBranchCulling=true, weights [1,1]}}
 {id=Random_4, children [WavePlayer_0,WavePlayer_1,WavePlayer_2,WavePlayer_3], properties {weights [1,1,1,1]}}
 {id=WavePlayer_0, properties {soundWaveAssetPtr=/Game/Sounds/Fort_Foley_Sounds/Moves/player_run_foley_15}}
 {id=WavePlayer_1, properties {soundWaveAssetPtr=/Game/Sounds/Fort_Foley_Sounds/Moves/player_run_foley_14}}
 {id=WavePlayer_2, properties {soundWaveAssetPtr=/Game/Sounds/Fort_Foley_Sounds/Moves/player_run_foley_13}}
 {id=WavePlayer_3, properties {soundWaveAssetPtr=/Game/Sounds/Fort_Foley_Sounds/Moves/player_run_foley_12}}
 {id=WavePlayer_4, properties {soundWaveAssetPtr=&/EMP_Grenade_Cook_02}}
 {id=WavePlayer_5, properties {soundWaveAssetPtr=&/EMP_Grenade_Cook_01}}
 {id=WavePlayer_6, properties {soundWaveAssetPtr=&/EMP_Grenade_Cook_Glass_01}}
 {id=WavePlayer_7, properties {soundWaveAssetPtr=&/EMP_Grenade_Cook_Glass_02}}
 {id=WavePlayer_8, properties {soundWaveAssetPtr=&/EMP_Grenade_Cook_Glass_03}}
]
properties {
 attenuationSettings=/Game/Sounds/Fort_Weapons/Grenades/Grenade_Default_Cook_Att
 bOverrideConcurrency=true
 maxDistance=3000
 pitchMultiplier=1.02
 soundClassObject=&Fort_Gadgets_Items
 volumeMultiplier=0.2
 attenuationOverrides {falloffDistance=1500, nonSpatializedRadiusStart=200, attenuationShapeExtents {x=1200}}
 concurrencyOverrides {maxCount=1, resolutionRule=StopOldest}
}
