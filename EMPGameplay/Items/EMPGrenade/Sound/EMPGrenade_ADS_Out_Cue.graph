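# Sound cue for the EMP grenade, named for leaving aim-down-sights, classed as weapon foley.
# A mixer layers two random picks: a player run foley clip with varied raised pitch, and a volume-boosted EMP grenade cook sound.
# Attenuation is overridden, and concurrency allows a single instance, stopping the oldest.
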SoundCue /EMPGameplay/Items/EMPGrenade/Sound/EMPGrenade_ADS_Out_Cue
root=Mixer_0
nodes [
 {id=Mixer_0, children [Modulator_0,Modulator_2], properties {inputVolume [0.3,0.3]}}
 {id=Modulator_0, children [Random_2], properties {pitchMax=1.2, pitchMin=1}}
 {id=Modulator_2, children [Random_0], properties {pitchMax=1, volumeMax=2.5, volumeMin=2.5}}
 {id=Random_0, children [WavePlayer_4,WavePlayer_5], properties {weights [1,1]}}
 {id=Random_2, children [WavePlayer_2,WavePlayer_3], properties {weights [1,1]}}
 {id=WavePlayer_2, properties {soundWaveAssetPtr=/Game/Sounds/Fort_Foley_Sounds/Moves/player_run_foley_01}}
 {id=WavePlayer_3, properties {soundWaveAssetPtr=/Game/Sounds/Fort_Foley_Sounds/Moves/player_run_foley_05}}
 {id=WavePlayer_4, properties {soundWaveAssetPtr=&/EMP_Grenade_Cook_02}}
 {id=WavePlayer_5, properties {soundWaveAssetPtr=&/EMP_Grenade_Cook_01}}
]
properties {
 attenuationSettings=&Fort_Weapon_Projectile_FirstPerson_Athena_Att
 bOverrideAttenuation=true
 bOverrideConcurrency=true
 maxDistance=2700
 soundClassObject=&Fort_WeaponFoley
 volumeMultiplier=0.15
 attenuationOverrides {falloffDistance=1500, nonSpatializedRadiusStart=200, attenuationShapeExtents {x=1200}}
 concurrencyOverrides {maxCount=1, resolutionRule=StopOldest}
}
