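# Sound cue for the EMP grenade's looping disable sound, played through the weapons sound class with its own attenuation.
# A quality switch sends its first two levels to a random pick of ten looping waves and its last two to just three of them.
# The pick passes through a modulator for pitch and volume variation.
# The cue uses the shared weapon explosion concurrency set.
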SoundCue /EMPGameplay/Items/EMPGrenade/Sound/EMPGrenade_DisableLoops_Cue
root=Modulator_2
nodes [
 {id=Modulator_2, children [QualityLevel_0]}
 {id=QualityLevel_0, children [Random_3,Random_3,Random_8,Random_8]}
 {
  id=Random_3
  children [WavePlayer_0,WavePlayer_1,WavePlayer_2,WavePlayer_3,WavePlayer_4,WavePlayer_5,WavePlayer_6,WavePlayer_7,WavePlayer_8,WavePlayer_9]
  properties {weights [1,1,1,1,1,1,1,1,1,1]}
 }
 {id=Random_8, children [WavePlayer_7,WavePlayer_8,WavePlayer_9], properties {bShouldExcludeFromBranchCulling=true, weights [1,1,1]}}
 {id=WavePlayer_0, properties {bLooping=true, soundWaveAssetPtr=&/EMP_Grenade_DisableLoop_01}}
 {id=WavePlayer_1, properties {bLooping=true, soundWaveAssetPtr=&/EMP_Grenade_DisableLoop_02}}
 {id=WavePlayer_2, properties {bLooping=true, soundWaveAssetPtr=&/EMP_Grenade_DisableLoop_03}}
 {id=WavePlayer_3, properties {bLooping=true, soundWaveAssetPtr=&/EMP_Grenade_DisableLoop_04}}
 {id=WavePlayer_4, properties {bLooping=true, soundWaveAssetPtr=&/EMP_Grenade_DisableLoop_05}}
 {id=WavePlayer_5, properties {bLooping=true, soundWaveAssetPtr=&/EMP_Grenade_DisableLoop_06}}
 {id=WavePlayer_6, properties {bLooping=true, soundWaveAssetPtr=&/EMP_Grenade_DisableLoop_07}}
 {id=WavePlayer_7, properties {bLooping=true, soundWaveAssetPtr=&/EMP_Grenade_DisableLoop_08}}
 {id=WavePlayer_8, properties {bLooping=true, soundWaveAssetPtr=&/EMP_Grenade_DisableLoop_09}}
 {id=WavePlayer_9, properties {bLooping=true, soundWaveAssetPtr=&/EMP_Grenade_DisableLoop_10}}
]
properties {
 attenuationSettings=/EMPGameplay/Items/EMPGrenade/Sound/Settings/EMPGrenade_DisableLoops_Att
 maxDistance=6000
 soundClassObject=/Game/Sounds/FortSoundClassesAndModes/Fort_Weapons
 volumeMultiplier=0.8
 concurrencyOverrides {maxCount=4, resolutionRule=StopOldest, volumeScale=0.5}
 concurrencySet [/Game/Sounds/Concurrency_Rules/Weapon_Explosion_Conc]
}
