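# Sound cue for an EMP grenade dropping and landing, layering impact and electrical sounds at reduced overall volume.
# A root mixer plays one of four random drop waves immediately, alongside a slightly delayed electrical layer.
# That layer mixes a random electric equip wave with a random explosion zap reused at raised pitch, each through its own modulator.
# It uses shared weapon-drop attenuation and concurrency, with the loot-collect sound class.
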SoundCue /EMPGameplay/Items/EMPGrenade/Sound/EMPGrenade_DropLand_Cue
root=Mixer_0
nodes [
 {id=Delay_0, children [Mixer_1], properties {delayMax=0.1, delayMin=0.1}}
 {id=Mixer_0, children [Modulator_2,Delay_0], properties {inputVolume [1,1]}}
 {id=Mixer_1, children [Modulator_0,Modulator_5], properties {inputVolume [1,1]}}
 {id=Modulator_0, children [Random_1]}
 {id=Modulator_2, children [Random_0]}
 {id=Modulator_5, children [Random_2], properties {pitchMax=1.5, pitchMin=1.5}}
 {id=Random_0, children [WavePlayer_0,WavePlayer_1,WavePlayer_2,WavePlayer_3], properties {weights [1,1,1,1]}}
 {id=Random_1, children [WavePlayer_9, WavePlayer_10], properties {weights [1,1]}}
 {id=Random_2, children [WavePlayer_4,WavePlayer_5,WavePlayer_6], properties {weights [1,1,1]}}
 {id=WavePlayer_0, properties {soundWaveAssetPtr=&/EMP_Grenade_Drop_01}}
 {id=WavePlayer_1, properties {soundWaveAssetPtr=&/EMP_Grenade_Drop_02}}
 {id=WavePlayer_10, properties {soundWaveAssetPtr=&/EMP_Grenade_ElectricEquip_02}}
 {id=WavePlayer_2, properties {soundWaveAssetPtr=&/EMP_Grenade_Drop_03}}
 {id=WavePlayer_3, properties {soundWaveAssetPtr=&/EMP_Grenade_Drop_04}}
 {id=WavePlayer_4, properties {soundWaveAssetPtr=&/EMP_Grenade_Explode_Zaps_01}}
 {id=WavePlayer_5, properties {soundWaveAssetPtr=&/EMP_Grenade_Explode_Zaps_02}}
 {id=WavePlayer_6, properties {soundWaveAssetPtr=&/EMP_Grenade_Explode_Zaps_03}}
 {id=WavePlayer_9, properties {soundWaveAssetPtr=&/EMP_Grenade_ElectricEquip_01}}
]
properties {
 attenuationSettings=&Fort_Weapon_Drop_Att
 maxDistance=3000
 soundClassObject=&Fort_Loot_Collect
 volumeMultiplier=0.3
 concurrencyOverrides {resolutionRule=PreventNew}
 concurrencySet [&Weapon_Drop_StopOldest]
}
