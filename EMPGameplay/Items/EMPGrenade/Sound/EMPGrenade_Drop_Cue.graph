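# Sound cue for dropping the EMP grenade item, built from the four EMP grenade pickup sound waves.
# A Random node picks one wave with equal weighting and feeds a Modulator at the root.
# The cue reuses pickup audio at raised pitch and reduced volume.
# It uses shared weapon-drop attenuation and concurrency assets, with the loot-collect sound class.
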SoundCue /EMPGameplay/Items/EMPGrenade/Sound/EMPGrenade_Drop_Cue
root=Modulator_1
nodes [
 {id=Modulator_1, children [Random_1]}
 {id=Random_1, children [WavePlayer_4,WavePlayer_5,WavePlayer_6,WavePlayer_7], properties {weights [1,1,1,1]}}
 {id=WavePlayer_4, properties {soundWaveAssetPtr=&/EMP_Grenade_PickUp_04}}
 {id=WavePlayer_5, properties {soundWaveAssetPtr=&/EMP_Grenade_PickUp_01}}
 {id=WavePlayer_6, properties {soundWaveAssetPtr=&/EMP_Grenade_PickUp_02}}
 {id=WavePlayer_7, properties {soundWaveAssetPtr=&/EMP_Grenade_PickUp_03}}
]
properties {
 attenuationSettings=&Fort_Weapon_Drop_Att
 maxDistance=3000
 pitchMultiplier=1.5
 soundClassObject=&Fort_Loot_Collect
 volumeMultiplier=0.6
 concurrencyOverrides {resolutionRule=PreventNew}
 concurrencySet [&Weapon_Drop_StopOldest]
}
