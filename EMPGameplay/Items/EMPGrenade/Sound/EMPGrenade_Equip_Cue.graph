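# Sound cue for equipping the EMP grenade, layering randomized equip, electric, catch and pickup sounds into one timed sequence.
# An equip variant starts at once, with a quieter electric layer slightly delayed, followed half a second in by a catch plus low-level pickup mix.
# Each layer passes a random wave through a modulator.
# It uses the idle loop's attenuation and overrides concurrency to one instance, stopping the oldest.
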SoundCue /EMPGameplay/Items/EMPGrenade/Sound/EMPGrenade_Equip_Cue
root=Mixer_0
nodes [
 {id=Delay_0, children [Mixer_1], properties {delayMax=0.5, delayMin=0.5}}
 {id=Delay_1, children [Modulator_3], properties {delayMax=0.15, delayMin=0.15}}
 {id=Mixer_0, children [Mixer_2,Delay_0], properties {inputVolume [1,1]}}
 {id=Mixer_1, children [Modulator_1,Modulator_2], properties {inputVolume [1,0.2]}}
 {id=Mixer_2, children [Modulator_0,Delay_1], properties {inputVolume [1,0.5]}}
 {id=Modulator_0, children [Random_0]}
 {id=Modulator_1, children [Random_1]}
 {id=Modulator_2, children [Random_2], properties {volumeMax=0.8, volumeMin=0.8}}
 {id=Modulator_3, children [Random_3]}
 {id=Random_0, children [WavePlayer_0,WavePlayer_1,WavePlayer_2], properties {weights [1,1,1]}}
 {id=Random_1, children [WavePlayer_3,WavePlayer_4,WavePlayer_5], properties {weights [1,1,1]}}
 {id=Random_2, children [WavePlayer_6,WavePlayer_7,WavePlayer_8], properties {weights [1,1,1]}}
 {id=Random_3, children [WavePlayer_9, WavePlayer_10], properties {weights [1,1]}}
 {id=WavePlayer_0, properties {soundWaveAssetPtr=&/EMP_Grenade_Equip_01}}
 {id=WavePlayer_1, properties {soundWaveAssetPtr=&/EMP_Grenade_Equip_02}}
 {id=WavePlayer_10, properties {soundWaveAssetPtr=&/EMP_Grenade_ElectricEquip_02}}
 {id=WavePlayer_2, properties {soundWaveAssetPtr=&/EMP_Grenade_Equip_03}}
 {id=WavePlayer_3, properties {soundWaveAssetPtr=&/EMP_Grenade_Equip_Catch_01}}
 {id=WavePlayer_4, properties {soundWaveAssetPtr=&/EMP_Grenade_Equip_Catch_02}}
 {id=WavePlayer_5, properties {soundWaveAssetPtr=&/EMP_Grenade_Equip_Catch_03}}
 {id=WavePlayer_6, properties {soundWaveAssetPtr=&/EMP_Grenade_PickUp_01}}
 {id=WavePlayer_7, properties {soundWaveAssetPtr=&/EMP_Grenade_PickUp_02}}
 {id=WavePlayer_8, properties {soundWaveAssetPtr=&/EMP_Grenade_PickUp_04}}
 {id=WavePlayer_9, properties {soundWaveAssetPtr=&/EMP_Grenade_ElectricEquip_01}}
]
properties {
 attenuationSettings=/EMPGameplay/Items/EMPGrenade/Sound/Settings/EMPGrenade_Idle_Loop_Att
 bOverrideConcurrency=true
 maxDistance=6000
 soundClassObject=/Game/Sounds/FortSoundClassesAndModes/Fort_Weapons
 volumeMultiplier=0.6
 concurrencyOverrides {maxCount=1, resolutionRule=StopOldest, volumeScale=0.5}
 concurrencySet [/Game/Sounds/Concurrency_Rules/Weapon_Explosion_Conc]
}
