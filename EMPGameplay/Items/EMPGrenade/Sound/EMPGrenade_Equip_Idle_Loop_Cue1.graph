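# Looping Sound Cue for the EMP grenade's equipped idle sound, spatialized through its own attenuation asset.
# A Mixer blends a modulated idle loop, a container glass loop, and container one-shots that repeat after random delays at lowered pitch.
# A Quality Level node switches the one-shots between seven-clip and three-clip random pools.
# An Enveloper over the whole mix swells the volume early, then gradually fades it to a quieter level.
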SoundCue /EMPGameplay/Items/EMPGrenade/Sound/EMPGrenade_Equip_Idle_Loop_Cue1
root=Enveloper_0
nodes [
 {id=Delay_0, children [Modulator_0], properties {delayMax=4, delayMin=2}}
 {
  id=Enveloper_0
  children [Mixer_0]
  properties {
   pitchCurve {editorCurveData {keys [{time=0}]}}
   volumeCurve {editorCurveData {keys [
      {interpMode=RCIM_Cubic, tangentMode=RCTM_User, time=0, value=0.5}
      {arriveTangent=-0.0137, interpMode=RCIM_Cubic, leaveTangent=-0.0137, tangentMode=RCTM_User, time=1.5, value=1}
      {arriveTangent=-0.0283, interpMode=RCIM_Cubic, leaveTangent=-0.0283, tangentMode=RCTM_User, time=8, value=0.8846}
      {arriveTangent=-0.1053, interpMode=RCIM_Cubic, leaveTangent=-0.1053, tangentMode=RCTM_User, time=15, value=0.45}
     ]}}
  }
 }
 {id=Looping_0, children [Delay_0]}
 {id=Mixer_0, children [Modulator_3, Looping_0, WavePlayer_15], properties {inputVolume [1.1,0.3,2.5]}}
 {id=Modulator_0, children [QualityLevel_0], properties {pitchMax=0.9, pitchMin=0.85}}
 {id=Modulator_3, children [WavePlayer_7]}
 {id=QualityLevel_0, children [Random_0,Random_0,Random_2,Random_2]}
 {
  id=Random_0
  children [WavePlayer_6, WavePlayer_8, WavePlayer_9, WavePlayer_10, WavePlayer_11, WavePlayer_12, WavePlayer_13]
  properties {weights [1,1,1,1,1,1,1]}
 }
 {id=Random_2, children [WavePlayer_11, WavePlayer_12, WavePlayer_13], properties {bShouldExcludeFromBranchCulling=true, weights [1,1,1]}}
 {id=WavePlayer_10, properties {soundWaveAssetPtr=&/EMP_Grenade_Container_OneShots_04}}
 {id=WavePlayer_11, properties {soundWaveAssetPtr=&/EMP_Grenade_Container_OneShots_05}}
 {id=WavePlayer_12, properties {soundWaveAssetPtr=&/EMP_Grenade_Container_OneShots_06}}
 {id=WavePlayer_13, properties {soundWaveAssetPtr=&/EMP_Grenade_Container_OneShots_07}}
 {id=WavePlayer_15, properties {bLooping=true, soundWaveAssetPtr=&/EMP_Grenade_Container_Glass_Loop}}
 {id=WavePlayer_6, properties {soundWaveAssetPtr=&/EMP_Grenade_Container_OneShots_01}}
 {id=WavePlayer_7, properties {bLooping=true, soundWaveAssetPtr=&/EMP_Grenade_Idle_Loop}}
 {id=WavePlayer_8, properties {soundWaveAssetPtr=&/EMP_Grenade_Container_OneShots_02}}
 {id=WavePlayer_9, properties {soundWaveAssetPtr=&/EMP_Grenade_Container_OneShots_03}}
]
properties {
 attenuationSettings=/EMPGameplay/Items/EMPGrenade/Sound/Settings/EMPGrenade_Idle_Loop_Att
 maxDistance=6000
 soundClassObject=/Game/Sounds/FortSoundClassesAndModes/Fort_Weapons
 volumeMultiplier=1.2
 concurrencyOverrides {maxCount=4, resolutionRule=StopOldest, volumeScale=0.5}
 concurrencySet [/Game/Sounds/Concurrency_Rules/Weapon_Explosion_Conc]
}
