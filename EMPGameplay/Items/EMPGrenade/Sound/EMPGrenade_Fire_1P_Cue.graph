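# Sound cue for the first-person EMP grenade fire (throw) sound, routed to a first-person ranged weapon fire sound class.
# A mixer layers throw sounds with slightly quieter electrical zaps, each layer passing through a modulator for slight volume or pitch variation.
# Per layer, a quality switch picks randomly from six waves at its first two levels, or only the last three at the other two.
# It references grenade throw attenuation and first-person gunfire concurrency.
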SoundCue /EMPGameplay/Items/EMPGrenade/Sound/EMPGrenade_Fire_1P_Cue
root=Mixer_0
nodes [
 {id=Mixer_0, children [Modulator_2,Modulator_3], properties {inputVolume [0.7,1]}}
 {id=Modulator_2, children [QualityLevel_0], properties {volumeMax=1.1, volumeMin=1}}
 {id=Modulator_3, children [QualityLevel_1], properties {pitchMax=1, pitchMin=0.98, volumeMax=1, volumeMin=1}}
 {id=QualityLevel_0, children [Random_0,Random_0,Random_1,Random_1]}
 {id=QualityLevel_1, children [Random_2,Random_2,Random_3,Random_3]}
 {id=Random_0, children [WavePlayer_0,WavePlayer_1,WavePlayer_2,WavePlayer_3,WavePlayer_4,WavePlayer_5], properties {weights [1,1,1,1,1,1]}}
 {id=Random_1, children [WavePlayer_3,WavePlayer_4,WavePlayer_5], properties {bShouldExcludeFromBranchCulling=true, weights [1,1,1]}}
 {
  id=Random_2
  children [WavePlayer_6, WavePlayer_7, WavePlayer_8, WavePlayer_9, WavePlayer_10, WavePlayer_11]
  properties {weights [1,1,1,1,1,1]}
 }
 {id=Random_3, children [WavePlayer_9, WavePlayer_10, WavePlayer_11], properties {bShouldExcludeFromBranchCulling=true, weights [1,1,1]}}
 {id=WavePlayer_0, properties {soundWaveAssetPtr=&/EMP_Grenade_Throw_01}}
 {id=WavePlayer_1, properties {soundWaveAssetPtr=&/EMP_Grenade_Throw_02}}
 {id=WavePlayer_10, properties {soundWaveAssetPtr=&/EMP_Grenade_Throw_Zaps_05}}
 {id=WavePlayer_11, properties {soundWaveAssetPtr=&/EMP_Grenade_Throw_Zaps_06}}
 {id=WavePlayer_2, properties {soundWaveAssetPtr=&/EMP_Grenade_Throw_03}}
 {id=WavePlayer_3, properties {soundWaveAssetPtr=&/EMP_Grenade_Throw_04}}
 {id=WavePlayer_4, properties {soundWaveAssetPtr=&/EMP_Grenade_Throw_05}}
 {id=WavePlayer_5, properties {soundWaveAssetPtr=&/EMP_Grenade_Throw_06}}
 {id=WavePlayer_6, properties {soundWaveAssetPtr=&/EMP_Grenade_Throw_Zaps_01}}
 {id=WavePlayer_7, properties {soundWaveAssetPtr=&/EMP_Grenade_Throw_Zaps_02}}
 {id=WavePlayer_8, properties {soundWaveAssetPtr=&/EMP_Grenade_Throw_Zaps_03}}
 {id=WavePlayer_9, properties {soundWaveAssetPtr=&/EMP_Grenade_Throw_Zaps_04}}
]
properties {
 attenuationSettings=/EMPGameplay/Items/EMPGrenade/Sound/Settings/EMPGrenade_Throw_Att
 bDebug=true
 maxDistance=2000
 soundClassObject=&Fort_WeaponRanged_Fire_1P_PSM
 volumeMultiplier=0.8
 attenuationOverrides {bSpatialize=false}
 concurrencyOverrides {bLimitToOwner=true, maxCount=3, resolutionRule=StopOldest}
 concurrencySet [&Weapon_Gunfire_FirstPerson_Conc]
}
