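# Sound cue for the looping in-air sound of the EMP grenade.
# A looping wave player feeds a Modulator root, which has no changed settings.
# Playback uses a dedicated attenuation asset, the weapons sound class and a boosted volume.
# It references the shared weapon explosion concurrency rule and also carries override values that stop the oldest instance past a small count.
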SoundCue /EMPGameplay/Items/EMPGrenade/Sound/EMPGrenade_InAir_Loop_Cue
root=Modulator_2
nodes [
 {id=Modulator_2, children [WavePlayer_0]}
 {id=WavePlayer_0, properties {bLooping=true, soundWaveAssetPtr=&/EMP_Grenade_InAir_Loop_01}}
]
properties {
 attenuationSettings=/EMPGameplay/Items/EMPGrenade/Sound/Settings/EMPGrenade_InAir_Loop_Att
 maxDistance=5000
 soundClassObject=/Game/Sounds/FortSoundClassesAndModes/Fort_Weapons
 volumeMultiplier=2
 concurrencyOverrides {maxCount=4, resolutionRule=StopOldest, volumeScale=0.5}
 concurrencySet [/Game/Sounds/Concurrency_Rules/Weapon_Explosion_Conc]
}
