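# Sound cue for picking up the EMP Grenade item.
# A Random node picks one of four equally weighted pickup recordings and feeds it through a Modulator at the root.
# It uses the shared weapon pickup attenuation and loot-collect sound class, with cue volume and pitch lowered.
# Simultaneous playback is governed by the PickupMaximum_Oldest concurrency set.
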SoundCue /EMPGameplay/Items/EMPGrenade/Sound/EMPGrenade_Pickup_Cue
root=Modulator_0
nodes [
 {id=Modulator_0, children [Random_0]}
 {id=Random_0, children [WavePlayer_0,WavePlayer_1,WavePlayer_2,WavePlayer_3], properties {weights [1,1,1,1]}}
 {id=WavePlayer_0, properties {soundWaveAssetPtr=&/EMP_Grenade_PickUp_04}}
 {id=WavePlayer_1, properties {soundWaveAssetPtr=&/EMP_Grenade_PickUp_01}}
 {id=WavePlayer_2, properties {soundWaveAssetPtr=&/EMP_Grenade_PickUp_02}}
 {id=WavePlayer_3, properties {soundWaveAssetPtr=&/EMP_Grenade_PickUp_03}}
]
properties {
 attenuationSettings=&Fort_Weapon_Pickup_Att
 maxDistance=3500
 pitchMultiplier=0.9
 soundClassObject=&Fort_Loot_Collect
 volumeMultiplier=0.6
 concurrencyOverrides {maxCount=1, resolutionRule=PreventNew}
 concurrencySet [&PickupMaximum_Oldest]
}
