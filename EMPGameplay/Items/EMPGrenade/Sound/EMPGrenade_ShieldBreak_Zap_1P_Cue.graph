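# First-person sound cue for the EMP grenade's shield-break zap, routed to the gadget item sound class.
# After a fixed short delay, a branch on the IsLocal parameter plays audio only when true, with the false and unset inputs left empty.
# That path runs through a neutral modulator into an evenly weighted random pick of four 2D zap waves.
# It references a dedicated 1P attenuation asset and the first-person weapon gunfire concurrency set.
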SoundCue /EMPGameplay/Items/EMPGrenade/Sound/EMPGrenade_ShieldBreak_Zap_1P_Cue
root=Delay_0
nodes [
 {id=Branch_0, children [Modulator_2,null,null], properties {boolParameterName=IsLocal}}
 {id=Delay_0, children [Branch_0], properties {delayMax=0.15, delayMin=0.15}}
 {id=Modulator_2, children [Random_0], properties {pitchMax=1, pitchMin=1, volumeMax=1, volumeMin=1}}
 {id=Random_0, children [WavePlayer_12, WavePlayer_13, WavePlayer_14, WavePlayer_15], properties {weights [1,1,1,1]}}
 {id=WavePlayer_12, properties {soundWaveAssetPtr=&/EMP_Grenade_2D_Zaps_01}}
 {id=WavePlayer_13, properties {soundWaveAssetPtr=&/EMP_Grenade_2D_Zaps_02}}
 {id=WavePlayer_14, properties {soundWaveAssetPtr=&/EMP_Grenade_2D_Zaps_03}}
 {id=WavePlayer_15, properties {soundWaveAssetPtr=&/EMP_Grenade_2D_Zaps_04}}
]
properties {
 attenuationSettings=/EMPGameplay/Items/EMPGrenade/Sound/Settings/EMPGrenade_Shield_Break_Zaps_1P_Att
 bDebug=true
 maxDistance=2097152
 soundClassObject=&Fort_Gadgets_Items
 volumeMultiplier=1
 attenuationOverrides {bSpatialize=false}
 concurrencyOverrides {bLimitToOwner=true, maxCount=3, resolutionRule=StopOldest}
 concurrencySet [&Weapon_Gunfire_FirstPerson_Conc]
}
