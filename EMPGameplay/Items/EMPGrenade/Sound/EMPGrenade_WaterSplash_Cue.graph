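# Sound cue for the EMP grenade's water splash.
# It picks at random among five equally weighted splash waves, delays playback by a fixed quarter second, then raises pitch through a modulator with no random variation.
# It uses a dedicated splash attenuation asset, overrides concurrency to cap simultaneous instances, and routes to a third-person ranged weapon fire sound class.
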SoundCue /EMPGameplay/Items/EMPGrenade/Sound/EMPGrenade_WaterSplash_Cue
root=Modulator_2
nodes [
 {id=Delay_0, children [Random_0], properties {delayMax=0.25, delayMin=0.25}}
 {id=Modulator_2, children [Delay_0], properties {pitchMax=1.2, pitchMin=1.2, volumeMax=1, volumeMin=1}}
 {id=Random_0, children [WavePlayer_0,WavePlayer_1,WavePlayer_2,WavePlayer_3,WavePlayer_4], properties {weights [1,1,1,1,1]}}
 {id=WavePlayer_0, properties {soundWaveAssetPtr=&/EMP_Grenade_WaterSplash_01}}
 {id=WavePlayer_1, properties {soundWaveAssetPtr=&/EMP_Grenade_WaterSplash_02}}
 {id=WavePlayer_2, properties {soundWaveAssetPtr=&/EMP_Grenade_WaterSplash_03}}
 {id=WavePlayer_3, properties {soundWaveAssetPtr=&/EMP_Grenade_WaterSplash_04}}
 {id=WavePlayer_4, properties {soundWaveAssetPtr=&/EMP_Grenade_WaterSplash_05}}
]
properties {
 attenuationSettings=/EMPGameplay/Items/EMPGrenade/Sound/Settings/EMPGrenade_WaterSplash_Att
 bOverrideConcurrency=true
 maxDistance=6000
 pitchMultiplier=1.02
 soundClassObject=&Fort_WeaponRanged_Fire_3P_PSM
 volumeMultiplier=1
 attenuationOverrides {falloffDistance=10000, attenuationShapeExtents {x=500}}
 concurrencyOverrides {maxCount=3}
}
