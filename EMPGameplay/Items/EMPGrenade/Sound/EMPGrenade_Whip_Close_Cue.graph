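# Sound Cue for the EMP grenade's close whip-by, reusing the shared bullet-whip attenuation and whiz-by concurrency rules.
# It mixes a whip layer, slightly pitch-randomized at or below normal, with a higher-pitched throw-zap layer shaped by a rising pitch envelope.
# In each layer a quality-level node selects either a random pick from six waves or from a three-wave subset, which is excluded from branch culling.
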SoundCue /EMPGameplay/Items/EMPGrenade/Sound/EMPGrenade_Whip_Close_Cue
root=Mixer_0
nodes [
 {
  id=Enveloper_0
  children [Modulator_0]
  properties {pitchCurve {editorCurveData {keys [{time=0}, <at 0.2455 0.8232>]}}, volumeCurve {editorCurveData {keys [<at 0 1>]}}}
 }
 {id=Mixer_0, children [Modulator_1,Enveloper_0], properties {inputVolume [0.7,1.25]}}
 {id=Modulator_0, children [QualityLevel_0], properties {pitchMax=1.4, pitchMin=1.3}}
 {id=Modulator_1, children [QualityLevel_1], properties {pitchMax=1, pitchMin=0.9}}
 {id=QualityLevel_0, children [Random_0,Random_0,Random_1,Random_1]}
 {id=QualityLevel_1, children [Random_2,Random_2,Random_3,Random_3]}
 {
  id=Random_0
  children [WavePlayer_6, WavePlayer_7, WavePlayer_8, WavePlayer_9, WavePlayer_10, WavePlayer_11]
  properties {weights [1,1,1,1,1,1]}
 }
 {id=Random_1, children [WavePlayer_9, WavePlayer_10, WavePlayer_11], properties {bShouldExcludeFromBranchCulling=true, weights [1,1,1]}}
 {id=Random_2, children [WavePlayer_0,WavePlayer_1,WavePlayer_2,WavePlayer_3,WavePlayer_4,WavePlayer_5], properties {weights [1,1,1,1,1,1]}}
 {id=Random_3, children [WavePlayer_3,WavePlayer_4,WavePlayer_5], properties {bShouldExcludeFromBranchCulling=true, weights [1,1,1]}}
 {id=WavePlayer_0, properties {soundWaveAssetPtr=&/EMP_Grenade_Whip_Close_06}}
 {id=WavePlayer_1, properties {soundWaveAssetPtr=&/EMP_Grenade_Whip_Close_01}}
 {id=WavePlayer_10, properties {soundWaveAssetPtr=&/EMP_Grenade_Throw_Zaps_04}}
 {id=WavePlayer_11, properties {soundWaveAssetPtr=&/EMP_Grenade_Throw_Zaps_05}}
 {id=WavePlayer_2, properties {soundWaveAssetPtr=&/EMP_Grenade_Whip_Close_02}}
 {id=WavePlayer_3, properties {soundWaveAssetPtr=&/EMP_Grenade_Whip_Close_03}}
 {id=WavePlayer_4, properties {soundWaveAssetPtr=&/EMP_Grenade_Whip_Close_04}}
 {id=WavePlayer_5, properties {soundWaveAssetPtr=&/EMP_Grenade_Whip_Close_05}}
 {id=WavePlayer_6, properties {soundWaveAssetPtr=&/EMP_Grenade_Throw_Zaps_06}}
 {id=WavePlayer_7, properties {soundWaveAssetPtr=&/EMP_Grenade_Throw_Zaps_01}}
 {id=WavePlayer_8, properties {soundWaveAssetPtr=&/EMP_Grenade_Throw_Zaps_02}}
 {id=WavePlayer_9, properties {soundWaveAssetPtr=&/EMP_Grenade_Throw_Zaps_03}}
]
properties {
 attenuationSettings=/Game/Athena/Sounds/Weapons/BulletWhips/bullet_whip_Att.Bullet_Whip_Att
 bOverrideAttenuation=true
 maxDistance=4000
 priority=2
 soundClassObject=&Fort_WeaponRanged_Projectile_Effects
 volumeMultiplier=1
 attenuationOverrides {nonSpatializedRadiusStart=200, stereoSpread=500}
 concurrencySet [/Game/Sounds/Concurrency_Rules/Athena_BulletWhizBy_Conc]
}
