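# Sound Cue for the distant whip of the EMP grenade, classed as a ranged-weapon projectile effect.
# A Quality Level node sends its first two inputs to an equal-weight Random over six distant whip waves and its last two to a Random over waves 04–06, excluded from branch culling.
# A Modulator adds slight downward pitch variation atop reduced cue pitch and volume.
# It borrows the bullet whip attenuation and bullet whiz-by concurrency assets.
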SoundCue /EMPGameplay/Items/EMPGrenade/Sound/EMPGrenade_Whip_Far_Cue
root=Modulator_1
nodes [
 {id=Modulator_1, children [QualityLevel_1], properties {pitchMax=1, pitchMin=0.9}}
 {id=QualityLevel_1, children [Random_2,Random_2,Random_3,Random_3]}
 {id=Random_2, children [WavePlayer_0,WavePlayer_1,WavePlayer_2,WavePlayer_3,WavePlayer_4,WavePlayer_5], properties {weights [1,1,1,1,1,1]}}
 {id=Random_3, children [WavePlayer_3,WavePlayer_4,WavePlayer_5], properties {bShouldExcludeFromBranchCulling=true, weights [1,1,1]}}
 {id=WavePlayer_0, properties {soundWaveAssetPtr=&/EMP_Grenade_Whip_Distant_01}}
 {id=WavePlayer_1, properties {soundWaveAssetPtr=&/EMP_Grenade_Whip_Distant_02}}
 {id=WavePlayer_2, properties {soundWaveAssetPtr=&/EMP_Grenade_Whip_Distant_03}}
 {id=WavePlayer_3, properties {soundWaveAssetPtr=&/EMP_Grenade_Whip_Distant_04}}
 {id=WavePlayer_4, properties {soundWaveAssetPtr=&/EMP_Grenade_Whip_Distant_05}}
 {id=WavePlayer_5, properties {soundWaveAssetPtr=&/EMP_Grenade_Whip_Distant_06}}
]
properties {
 attenuationSettings=/Game/Athena/Sounds/Weapons/BulletWhips/bullet_whip_Att.Bullet_Whip_Att
 maxDistance=2097152
 pitchMultiplier=0.8
 soundClassObject=&Fort_WeaponRanged_Projectile_Effects
 volumeMultiplier=0.5
 attenuationOverrides {bAttenuate=false, stereoSpread=0}
 concurrencySet [/Game/Sounds/Concurrency_Rules/Athena_BulletWhizBy_Conc]
}
